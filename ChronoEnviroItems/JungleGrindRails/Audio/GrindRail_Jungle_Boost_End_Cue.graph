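# Sound cue for the end of a boost on jungle grind rails, layering a vine-specific sound over the generic grind rail boost end.
# An IsLocal branch picks first-person or third-person grind attenuation, with an unset parameter also using third-person.
# Both paths mix the same wave sets: a random vine variant at full volume and a random generic variant at reduced volume.
# Each path randomly modulates pitch downward for variation.
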SoundCue /ChronoEnviroItems/JungleGrindRails/Audio/GrindRail_Jungle_Boost_End_Cue
root=Branch_0
nodes [
 {id=Attenuation_0, children [Modulator_0], properties {attenuationSettings=/GrindRail/Audio/Grinding_1P_Att}}
 {id=Attenuation_1, children [Modulator_1], properties {attenuationSettings=/GrindRail/Audio/Grinding_3P_Loops_Att}}
 {id=Branch_0, children [Attenuation_0, Attenuation_1, Attenuation_1], properties {boolParameterName=IsLocal}}
 {id=Mixer_0, children [Random_2,Random_0], properties {inputVolume [1,0.4]}}
 {id=Mixer_1, children [Random_1,Random_3], properties {inputVolume [1,0.4]}}
 {id=Modulator_0, children [Mixer_0], properties {pitchMax=1, pitchMin=0.8}}
 {id=Modulator_1, children [Mixer_1], properties {pitchMax=1, pitchMin=0.8}}
 {id=Random_0, children [WavePlayer_4,WavePlayer_5,WavePlayer_6], properties {bShouldExcludeFromBranchCulling=true, weights [1,1,1]}}
 {id=Random_1, children [WavePlayer_0,WavePlayer_1,WavePlayer_2], properties {bShouldExcludeFromBranchCulling=true, weights [1,1,1]}}
 {id=Random_2, children [WavePlayer_3,WavePlayer_7,WavePlayer_8], properties {weights [1,1,1]}}
 {id=Random_3, children [WavePlayer_9, WavePlayer_10, WavePlayer_11], properties {bShouldExcludeFromBranchCulling=true, weights [1,1,1]}}
 {id=WavePlayer_0, properties {soundWaveAssetPtr=&/Vine_Rail_Boost_End_3p_02}}
 {id=WavePlayer_1, properties {soundWaveAssetPtr=&/Vine_Rail_Boost_End_3p_03}}
 {id=WavePlayer_10, properties {soundWaveAssetPtr=/GrindRail/Audio/Grind_Rail_Boost_End}}
 {id=WavePlayer_11, properties {soundWaveAssetPtr=/GrindRail/Audio/Grind_Rail_Boost_End_02}}
 {id=WavePlayer_2, properties {soundWaveAssetPtr=&/Vine_Rail_Boost_End_3p_01}}
 {id=WavePlayer_3, properties {soundWaveAssetPtr=&Vine_Rail_Boost_End_3p_02}}
 {id=WavePlayer_4, properties {soundWaveAssetPtr=/GrindRail/Audio/Grind_Rail_Boost_End_03}}
 {id=WavePlayer_5, properties {soundWaveAssetPtr=&Grind_Rail_Boost_End}}
 {id=WavePlayer_6, properties {soundWaveAssetPtr=&Grind_Rail_Boost_End_02}}
 {id=WavePlayer_7, properties {soundWaveAssetPtr=&Vine_Rail_Boost_End_3p_03}}
 {id=WavePlayer_8, properties {soundWaveAssetPtr=&Vine_Rail_Boost_End_3p_01}}
 {id=WavePlayer_9, properties {soundWaveAssetPtr=&Grind_Rail_Boost_End_03}}
]
properties {
 maxDistance=8500
 soundClassObject=&Fort_Gadgets
 volumeMultiplier=0.65
 attenuationOverrides {
  bEnableOcclusion=true
  lPFFrequencyAtMin=10000
  lPFRadiusMin=2000
  nonFocusVolumeAttenuation=0.7
  occlusionLowPassFilterFrequency=4000
  occlusionTraceChannel=ECC_GameTraceChannel11
  occlusionVolumeAttenuation=4
 }
 concurrencyOverrides {bLimitToOwner=true}
}
