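# Sound cue for boost footsteps on the jungle vine grind rail, routed through the gadgets sound class.
# An IsLocal branch picks first-person attenuation when true, and third-person attenuation when false or unset.
# Each path randomly plays one of the same six equally weighted vine energy footstep waves.
# Only the local path adds slight random pitch variation, and both random nodes are excluded from branch culling.
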SoundCue /ChronoEnviroItems/JungleGrindRails/Audio/GrindRail_Jungle_Boost_Steps_Cue
root=Branch_0
nodes [
 {id=Attenuation_0, children [Modulator_0], properties {attenuationSettings=/GrindRail/Audio/Grinding_1P_Att}}
 {id=Attenuation_1, children [Random_7], properties {attenuationSettings=/GrindRail/Audio/Grinding_3P_Att}}
 {id=Branch_0, children [Attenuation_0, Attenuation_1, Attenuation_1], properties {boolParameterName=IsLocal}}
 {id=Modulator_0, children [Random_3], properties {pitchMax=1.02, pitchMin=0.98}}
 {
  id=Random_3
  children [WavePlayer_6, WavePlayer_7, WavePlayer_8, WavePlayer_9, WavePlayer_10, WavePlayer_11]
  properties {bShouldExcludeFromBranchCulling=true, weights [1,1,1,1,1,1]}
 }
 {
  id=Random_7
  children [WavePlayer_0,WavePlayer_1,WavePlayer_2,WavePlayer_3,WavePlayer_4,WavePlayer_5]
  properties {bShouldExcludeFromBranchCulling=true, weights [1,1,1,1,1,1]}
 }
 {id=WavePlayer_0, properties {soundWaveAssetPtr=&/Vine_Rail_Footstep_Energy_01}}
 {id=WavePlayer_1, properties {soundWaveAssetPtr=&/Vine_Rail_Footstep_Energy_02}}
 {id=WavePlayer_10, properties {soundWaveAssetPtr=&/Vine_Rail_Footstep_Energy_05}}
 {id=WavePlayer_11, properties {soundWaveAssetPtr=&/Vine_Rail_Footstep_Energy_06}}
 {id=WavePlayer_2, properties {soundWaveAssetPtr=&/Vine_Rail_Footstep_Energy_03}}
 {id=WavePlayer_3, properties {soundWaveAssetPtr=&/Vine_Rail_Footstep_Energy_04}}
 {id=WavePlayer_4, properties {soundWaveAssetPtr=&Vine_Rail_Footstep_Energy_05}}
 {id=WavePlayer_5, properties {soundWaveAssetPtr=&Vine_Rail_Footstep_Energy_06}}
 {id=WavePlayer_6, properties {soundWaveAssetPtr=&Vine_Rail_Footstep_Energy_01}}
 {id=WavePlayer_7, properties {soundWaveAssetPtr=&Vine_Rail_Footstep_Energy_02}}
 {id=WavePlayer_8, properties {soundWaveAssetPtr=&Vine_Rail_Footstep_Energy_03}}
 {id=WavePlayer_9, properties {soundWaveAssetPtr=&Vine_Rail_Footstep_Energy_04}}
]
properties {maxDistance=6500, soundClassObject=&Fort_Gadgets, volumeMultiplier=0.65}
