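# Sound cue for landing a jump on a jungle vine grind rail, routed to the player foley sound class.
# An IsLocal branch picks first-person attenuation for the local player and third-person attenuation when false or unset.
# Both paths randomize modulated vine energy-connect waves, while only third-person mixes in a random vine landing layer.
# That landing layer's random node is excluded from branch culling, and the first-person mixer's second input is empty.
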SoundCue /ChronoEnviroItems/JungleGrindRails/Audio/GrindRail_Jungle_Jump_Land_Cue
root=Branch_0
nodes [
 {id=Attenuation_0, children [Mixer_1], properties {attenuationSettings=/GrindRail/Audio/Grinding_1P_Att}}
 {id=Attenuation_1, children [Mixer_2], properties {attenuationSettings=/GrindRail/Audio/Grinding_3P_Att}}
 {id=Branch_0, children [Attenuation_0, Attenuation_1, Attenuation_1], properties {boolParameterName=IsLocal}}
 {id=Mixer_1, children [Modulator_1,null], properties {inputVolume [1.2,1.5]}}
 {id=Mixer_2, children [Modulator_4,Modulator_5], properties {inputVolume [1.2,1.5]}}
 {id=Modulator_1, children [Random_1], properties {volumeMax=1.2}}
 {id=Modulator_4, children [Random_5], properties {volumeMax=1.2}}
 {id=Modulator_5, children [Random_6]}
 {id=Random_1, children [WavePlayer_0,WavePlayer_1], properties {weights [1,1]}}
 {id=Random_5, children [WavePlayer_2,WavePlayer_3], properties {weights [1,1]}}
 {id=Random_6, children [WavePlayer_11, WavePlayer_12], properties {bShouldExcludeFromBranchCulling=true, weights [1,1]}}
 {id=WavePlayer_0, properties {soundWaveAssetPtr=&/Vine_Rail_Jump_Land_Energy_Connect_01}}
 {id=WavePlayer_1, properties {soundWaveAssetPtr=&/Vine_Rail_Jump_Land_Energy_Connect_02}}
 {id=WavePlayer_11, properties {soundWaveAssetPtr=&/Grind_Land_On_Rail_Vine_01}}
 {id=WavePlayer_12, properties {soundWaveAssetPtr=&/Grind_Land_On_Rail_Vine_02}}
 {id=WavePlayer_2, properties {soundWaveAssetPtr=&Vine_Rail_Jump_Land_Energy_Connect_01}}
 {id=WavePlayer_3, properties {soundWaveAssetPtr=&Vine_Rail_Jump_Land_Energy_Connect_02}}
]
properties {maxDistance=6500, soundClassObject=&Fort_Player_Foley_Athena, volumeMultiplier=1.15}
