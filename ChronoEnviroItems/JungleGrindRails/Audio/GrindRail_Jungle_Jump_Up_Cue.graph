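# Sound cue for jumping off a jungle vine grind rail, in the player foley sound class.
# A branch on IsLocal sends local playback through first-person grinding attenuation, and non-local or unset playback through third-person attenuation at lower mix levels.
# Each path layers a random jump-off sound with a random thud, then modulates pitch for variation.
# A concurrency override allows one instance per owner with a retrigger delay.
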SoundCue /ChronoEnviroItems/JungleGrindRails/Audio/GrindRail_Jungle_Jump_Up_Cue
root=Branch_0
nodes [
 {id=Attenuation_0, children [Modulator_4], properties {attenuationSettings=/GrindRail/Audio/Grinding_1P_Att}}
 {id=Attenuation_1, children [Modulator_3], properties {attenuationSettings=/GrindRail/Audio/Grinding_3P_Att}}
 {id=Branch_0, children [Attenuation_0, Attenuation_1, Attenuation_1], properties {boolParameterName=IsLocal}}
 {id=Mixer_2, children [Random_6,Random_8], properties {inputVolume [1,1.5]}}
 {id=Mixer_3, children [Random_7,Random_9], properties {inputVolume [0.75,1.2]}}
 {id=Modulator_3, children [Mixer_3], properties {pitchMax=1.1}}
 {id=Modulator_4, children [Mixer_2], properties {pitchMax=1.1}}
 {
  id=Random_6
  children [WavePlayer_3,WavePlayer_5,WavePlayer_6,WavePlayer_7]
  properties {bShouldExcludeFromBranchCulling=true, weights [1,1,1,1]}
 }
 {
  id=Random_7
  children [WavePlayer_15, WavePlayer_16, WavePlayer_17, WavePlayer_18]
  properties {bShouldExcludeFromBranchCulling=true, weights [1,1,1,1]}
 }
 {id=Random_8, children [WavePlayer_9, WavePlayer_10, WavePlayer_11], properties {bShouldExcludeFromBranchCulling=true, weights [1,1,1]}}
 {id=Random_9, children [WavePlayer_12, WavePlayer_13, WavePlayer_14], properties {bShouldExcludeFromBranchCulling=true, weights [1,1,1]}}
 {id=WavePlayer_10, properties {soundWaveAssetPtr=&/Vine_Rail_Jump_Off_Rail_Thud_01}}
 {id=WavePlayer_11, properties {soundWaveAssetPtr=&/Vine_Rail_Jump_Off_Rail_Thud_02}}
 {id=WavePlayer_12, properties {soundWaveAssetPtr=&/Vine_Rail_Jump_Off_Rail_Thud_03}}
 {id=WavePlayer_13, properties {soundWaveAssetPtr=&Vine_Rail_Jump_Off_Rail_Thud_01}}
 {id=WavePlayer_14, properties {soundWaveAssetPtr=&Vine_Rail_Jump_Off_Rail_Thud_02}}
 {id=WavePlayer_15, properties {soundWaveAssetPtr=&/Vine_Rail_Jump_Off_Rail_04}}
 {id=WavePlayer_16, properties {soundWaveAssetPtr=&/Vine_Rail_Jump_Off_Rail_01}}
 {id=WavePlayer_17, properties {soundWaveAssetPtr=&/Vine_Rail_Jump_Off_Rail_02}}
 {id=WavePlayer_18, properties {soundWaveAssetPtr=&/Vine_Rail_Jump_Off_Rail_03}}
 {id=WavePlayer_3, properties {soundWaveAssetPtr=&Vine_Rail_Jump_Off_Rail_01}}
 {id=WavePlayer_5, properties {soundWaveAssetPtr=&Vine_Rail_Jump_Off_Rail_02}}
 {id=WavePlayer_6, properties {soundWaveAssetPtr=&Vine_Rail_Jump_Off_Rail_03}}
 {id=WavePlayer_7, properties {soundWaveAssetPtr=&Vine_Rail_Jump_Off_Rail_04}}
 {id=WavePlayer_9, properties {soundWaveAssetPtr=&Vine_Rail_Jump_Off_Rail_Thud_03}}
]
properties {
 bOverrideConcurrency=true
 maxDistance=6500
 soundClassObject=&Fort_Player_Foley_Athena
 volumeMultiplier=1
 concurrencyOverrides {bLimitToOwner=true, maxCount=1, retriggerTime=0.5}
}
