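# Sound cue for leaning forward on a jungle vine grind rail, classed as a gadget sound.
# A branch on the IsLocal parameter picks first-person grind attenuation when true, and third-person attenuation when false or unset.
# Each path randomly chooses between the same two vine lean waves at equal weight, with random volume and slightly lowered pitch.
# A concurrency override limits overlapping plays to two per owner.
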SoundCue /ChronoEnviroItems/JungleGrindRails/Audio/GrindRail_Jungle_Lean_Forward_Cue
root=Branch_0
nodes [
 {id=Attenuation_0, children [Modulator_0], properties {attenuationSettings=/GrindRail/Audio/Grinding_1P_Att}}
 {id=Attenuation_1, children [Modulator_1], properties {attenuationSettings=/GrindRail/Audio/Grinding_3P_Att}}
 {id=Branch_0, children [Attenuation_0, Attenuation_1, Attenuation_1], properties {boolParameterName=IsLocal}}
 {id=Modulator_0, children [Random_2], properties {pitchMax=0.99, pitchMin=0.8, volumeMax=1, volumeMin=0.65}}
 {id=Modulator_1, children [Random_3], properties {pitchMax=0.99, pitchMin=0.8, volumeMax=1, volumeMin=0.65}}
 {id=Random_2, children [WavePlayer_0,WavePlayer_1], properties {weights [1,1]}}
 {id=Random_3, children [WavePlayer_2,WavePlayer_3], properties {weights [1,1]}}
 {id=WavePlayer_0, properties {soundWaveAssetPtr=&/VineRail_LeanForward_04}}
 {id=WavePlayer_1, properties {soundWaveAssetPtr=&/VineRail_LeanForward_03}}
 {id=WavePlayer_2, properties {soundWaveAssetPtr=&VineRail_LeanForward_04}}
 {id=WavePlayer_3, properties {soundWaveAssetPtr=&VineRail_LeanForward_03}}
]
properties {
 bOverrideConcurrency=true
 maxDistance=6500
 soundClassObject=&Fort_Gadgets
 volumeMultiplier=0.504
 attenuationOverrides {
  bEnableOcclusion=true
  lPFFrequencyAtMin=10000
  lPFRadiusMin=2000
  nonFocusVolumeAttenuation=0.7
  occlusionLowPassFilterFrequency=4000
  occlusionTraceChannel=ECC_GameTraceChannel11
  occlusionVolumeAttenuation=4
 }
 concurrencyOverrides {bLimitToOwner=true, maxCount=2}
}
