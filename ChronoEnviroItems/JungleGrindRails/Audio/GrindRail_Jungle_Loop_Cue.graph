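# Sound cue for the looping sound of grinding on a jungle vine rail, with a fade-in through an enveloper.
# An IsLocal branch selects first-person or third-person attenuation and mix, both gated in volume by IsGrinding.
# The local mix layers base, cornering, wood, wind and boost loops, with continuous modulators driving pitch and volume from Speed, Curve and Boost.
# The simpler third-person mix, also the unset fallback, uses only base and boost loops.
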SoundCue /ChronoEnviroItems/JungleGrindRails/Audio/GrindRail_Jungle_Loop_Cue
root=Enveloper_0
nodes [
 {id=Attenuation_0, children [ModulatorContinuous_10], properties {attenuationSettings=/GrindRail/Audio/Grinding_1P_Att}}
 {id=Attenuation_1, children [ModulatorContinuous_47], properties {attenuationSettings=/GrindRail/Audio/Grinding_3P_Loops_Att}}
 {id=Branch_0, children [Attenuation_0, Attenuation_1, Attenuation_1], properties {boolParameterName=IsLocal}}
 {
  id=Enveloper_0
  children [Branch_0]
  properties {pitchCurve {editorCurveData {keys [{time=0}]}}, volumeCurve {editorCurveData {keys [<at 0 0>,<at 1 1>]}}}
 }
 {id=Mixer_0, children [ModulatorContinuous_0, ModulatorContinuous_3, ModulatorContinuous_8], properties {inputVolume [1,0.3,1]}}
 {id=Mixer_1, children [ModulatorContinuous_7, ModulatorContinuous_1], properties {inputVolume [0.75,1]}}
 {id=Mixer_10, children [ModulatorContinuous_51, null, ModulatorContinuous_13], properties {inputVolume [1.15,0.85,1.35]}}
 {id=Mixer_2, children [ModulatorContinuous_6, ModulatorContinuous_2], properties {inputVolume [1,1]}}
 {id=Mixer_3, children [Mixer_2, Mixer_0, Mixer_1, ModulatorContinuous_9], properties {inputVolume [0.55,1.8,3,0.8]}}
 {
  id=ModulatorContinuous_0
  children [ModulatorContinuous_36]
  properties {
   pitchModulationParams {maxOutput=1.25, minOutput=1, parameterName=Curve}
   volumeModulationParams {default=0, maxOutput=1.1, minOutput=0.2, parameterName=Curve}
  }
 }
 {
  id=ModulatorContinuous_1
  children [ModulatorContinuous_15]
  properties {
   pitchModulationParams {maxOutput=1.1, minOutput=1, parameterName=Curve}
   volumeModulationParams {default=0, maxOutput=0.2, minOutput=0.5, parameterName=Curve}
  }
 }
 {id=ModulatorContinuous_10, children [Mixer_3], properties {volumeModulationParams {default=0, parameterName=IsGrinding}}}
 {
  id=ModulatorContinuous_11
  children [WavePlayer_9]
  properties {
   pitchModulationParams {maxOutput=1.25, minOutput=0.65, parameterName=Speed}
   volumeModulationParams {default=0, maxOutput=0.75, minOutput=0.0025, parameterName=Speed}
  }
 }
 {
  id=ModulatorContinuous_12
  children [WavePlayer_6]
  properties {
   pitchModulationParams {maxOutput=1.15, minOutput=0.85, parameterName=Speed}
   volumeModulationParams {default=0, maxOutput=0.85, minOutput=0.1, parameterName=Speed}
  }
 }
 {
  id=ModulatorContinuous_13
  children [ModulatorContinuous_14]
  properties {pitchModulationParams {minOutput=1}, volumeModulationParams {default=0, parameterName=Boost}}
 }
 {
  id=ModulatorContinuous_14
  children [WavePlayer_7]
  properties {
   pitchModulationParams {maxOutput=1.15, minOutput=0.85, parameterName=Speed}
   volumeModulationParams {default=0, minOutput=0.2, parameterName=Speed}
  }
 }
 {
  id=ModulatorContinuous_15
  children [WavePlayer_5]
  properties {
   pitchModulationParams {maxOutput=1.1, minOutput=0.75, parameterName=Speed}
   volumeModulationParams {default=0, maxOutput=0.4, minOutput=0.3, parameterName=Speed}
  }
 }
 {
  id=ModulatorContinuous_16
  children [WavePlayer_2]
  properties {
   pitchModulationParams {minOutput=0.6, parameterName=Speed}
   volumeModulationParams {default=0, minInput=0.2, minOutput=0.0025, parameterName=Speed}
  }
 }
 {
  id=ModulatorContinuous_2
  children [ModulatorContinuous_11]
  properties {
   pitchModulationParams {maxOutput=1.2, minOutput=1, parameterName=Curve}
   volumeModulationParams {default=0, minOutput=0.75, parameterName=Curve}
  }
 }
 {
  id=ModulatorContinuous_3
  children [ModulatorContinuous_4]
  properties {
   pitchModulationParams {minOutput=0.9, parameterName=Curve}
   volumeModulationParams {default=0, maxOutput=0.45, minOutput=0.1, parameterName=Curve}
  }
 }
 {
  id=ModulatorContinuous_34
  children [WavePlayer_16]
  properties {
   pitchModulationParams {maxOutput=1.5, minOutput=0.85, parameterName=Speed}
   volumeModulationParams {default=0, maxOutput=0.5, minOutput=1, parameterName=Speed}
  }
 }
 {
  id=ModulatorContinuous_36
  children [WavePlayer_17]
  properties {
   pitchModulationParams {minOutput=0.6, parameterName=Speed}
   volumeModulationParams {default=0, minInput=0.2, minOutput=0.025, parameterName=Speed}
  }
 }
 {
  id=ModulatorContinuous_4
  children [WavePlayer_0]
  properties {
   pitchModulationParams {minOutput=0.6, parameterName=Speed}
   volumeModulationParams {default=0, minInput=0.2, minOutput=0.0025, parameterName=Speed}
  }
 }
 {id=ModulatorContinuous_47, children [Mixer_10], properties {volumeModulationParams {default=0, parameterName=IsGrinding}}}
 {
  id=ModulatorContinuous_5
  children [WavePlayer_1]
  properties {
   pitchModulationParams {maxOutput=1.05, minOutput=0.75, parameterName=Speed}
   volumeModulationParams {default=0, minOutput=0.01, parameterName=Speed}
  }
 }
 {
  id=ModulatorContinuous_51
  children [WavePlayer_22]
  properties {
   pitchModulationParams {maxOutput=1.35, minOutput=0.8, parameterName=Speed}
   volumeModulationParams {default=0, maxOutput=1.5, minOutput=0.25, parameterName=Speed}
  }
 }
 {
  id=ModulatorContinuous_6
  children [ModulatorContinuous_34]
  properties {pitchModulationParams {minOutput=1, parameterName=Curve}, volumeModulationParams {default=0, minOutput=0.75, parameterName=Curve}}
 }
 {
  id=ModulatorContinuous_7
  children [ModulatorContinuous_5]
  properties {
   pitchModulationParams {maxOutput=1.25, minOutput=1, parameterName=Curve}
   volumeModulationParams {default=0, maxOutput=1.1, minOutput=0.1, parameterName=Curve}
  }
 }
 {
  id=ModulatorContinuous_8
  children [ModulatorContinuous_16]
  properties {
   pitchModulationParams {minOutput=0.9, parameterName=Curve}
   volumeModulationParams {default=0, maxOutput=1.25, minOutput=0.05, parameterName=Curve}
  }
 }
 {
  id=ModulatorContinuous_9
  children [ModulatorContinuous_12]
  properties {pitchModulationParams {minOutput=1}, volumeModulationParams {default=0, maxOutput=1.1, parameterName=Boost}}
 }
 {id=WavePlayer_0, properties {bLooping=true, soundWaveAssetPtr=&/VineRail_Wood_Breaks_Loop}}
 {id=WavePlayer_1, properties {bLooping=true, soundWaveAssetPtr=&/VineRail_Wind_High_Loop}}
 {id=WavePlayer_16, properties {bLooping=true, soundWaveAssetPtr=&/Vine_Rail_Speed_Loop_Base_Layer01}}
 {id=WavePlayer_17, properties {bLooping=true, soundWaveAssetPtr=&/Vine_Rail_Speed_Loop_Cornering_01}}
 {id=WavePlayer_2, properties {bLooping=true, soundWaveAssetPtr=&/Vine_Rail_Wood_Creak_Stretch_Loop}}
 {
  id=WavePlayer_22
  properties {bLooping=true, soundWaveAssetPtr=&/Vine_Rail_Speed_Loop_Base_Layer01_3p}
 }
 {id=WavePlayer_5, properties {bLooping=true, soundWaveAssetPtr=/GrindRail/Audio/Grind_Rail_Speed_Loop_Wind_Buffet}}
 {id=WavePlayer_6, properties {bLooping=true, soundWaveAssetPtr=&/Vine_Rail_Speed_Boost_Loop_01}}
 {id=WavePlayer_7, properties {bLooping=true, soundWaveAssetPtr=&/Vine_Rail_Speed_Loop_Boost_3p}}
 {id=WavePlayer_9, properties {bLooping=true, soundWaveAssetPtr=&/Vine_Rail_Speed_Loop_Base_Layer02}}
]
properties {
 bOverrideConcurrency=true
 maxDistance=8500
 soundClassObject=&Fort_Gadgets
 volumeMultiplier=0.85
 attenuationOverrides {
  bEnableOcclusion=true
  lPFFrequencyAtMin=10000
  lPFRadiusMin=2000
  nonFocusVolumeAttenuation=0.7
  occlusionLowPassFilterFrequency=4000
  occlusionTraceChannel=ECC_GameTraceChannel11
  occlusionVolumeAttenuation=4
 }
 concurrencyOverrides {bVolumeScaleCanRelease=true, maxCount=4, voiceStealReleaseTime=1}
}
